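; Spotlight light can prop for the prototype Sparks stage show system, configured at construction time with optional light cone and flare.
; InitializeActor rebuilds the StaticMeshes array from the can plus whichever of beam and halo are enabled, toggling their visibility to match.
; It converts Heading and Pitch into a direction vector written to the can's default custom primitive data as LocalRestingDirection, then the parent construction script runs.
; Its BeginPlay, overlap and Tick events are empty.
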(event EventBeginPlay)

(event Collision|EventActorBeginOverlap
  (params (OtherActor :type Actor%20Object%20Reference)))

(event EventTick
  (params (DeltaSeconds :type Float%20%28single-precision%29)))

(fn InitializeActor
  (bind _enablelightcone (Variables|Default|GetEnableLightCone))
  (bind _enableflare (Variables|Default|GetEnableFlare))
  (Utilities|Array|Clear (Variables|Default|GetStaticMeshes))
  (Utilities|FlowControl|Sequence
    (:then_0
      (Utilities|Array|Add (Variables|Default|GetStaticMeshes) (Variables|Default|GetLightCan)))
    (:then_1
      (Rendering|SetVisibility (Variables|Default|GetLightBeam) _enablelightcone true)
      (if _enablelightcone
        (Utilities|Array|Add (Variables|Default|GetStaticMeshes) (Variables|Default|GetLightBeam))))
    (:then_2
      (Rendering|SetVisibility (Variables|Default|GetHalo) _enableflare true)
      (if _enableflare
        (Utilities|Array|Add (Variables|Default|GetStaticMeshes) (Variables|Default|GetHalo))))
    (:then_3
      (Rendering|Material|SetVectorParameterforDefaultCustomPrimitiveData (Variables|Default|GetLightCan) "LocalRestingDirection" (Math|Vector4|MakeVector4 (.x (Math|Vector|RotateVectorAroundAxis (Math|Vector|RotateVectorAroundAxis "1.000000,0.000000,0.000000" (Variables|Default|GetPitch) "0.000000,-1.000000,0.000000") (Variables|Default|GetHeading) "0.000000,0.000000,1.000000")) (.y (Math|Vector|RotateVectorAroundAxis (Math|Vector|RotateVectorAroundAxis "1.000000,0.000000,0.000000" (Variables|Default|GetPitch) "0.000000,-1.000000,0.000000") (Variables|Default|GetHeading) "0.000000,0.000000,1.000000")) (.z (Math|Vector|RotateVectorAroundAxis (Math|Vector|RotateVectorAroundAxis "1.000000,0.000000,0.000000" (Variables|Default|GetPitch) "0.000000,-1.000000,0.000000") (Variables|Default|GetHeading) "0.000000,0.000000,1.000000")))))))

(fn ConstructionScript
  (Default|InitializeActor)
  (|CallParentFunction))
